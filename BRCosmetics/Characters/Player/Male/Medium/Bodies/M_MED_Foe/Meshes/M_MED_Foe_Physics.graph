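# Physics asset for the M_MED_Foe male medium player body cosmetic, covering only the pelvis and three pumpkin attachments, likely for dangling secondary motion.
# Each pumpkin bone has a tiny capsule and joins the pelvis with free swing and twist.
# Each pumpkin carries a child "dyn" bone with a larger, offset capsule, constrained to a narrow swing on one axis with zero-degree limits on the others.
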PhysicsAsset /BRCosmetics/Characters/Player/Male/Medium/Bodies/M_MED_Foe/Meshes/M_MED_Foe_Physics
bodies {
 pelvis [{shapeType=Capsule, length=5, radius=5, shapeName=pelvis_capsule, center=<1.84 0.0256 0>, rotation=<rot 0.0004 -90 0.0002>}]
 pumpkin1 [{shapeType=Capsule, length=0.505, radius=0.505, shapeName=pumpkin1_capsule, rotation=<rot 0 -90 0>}]
 pumpkin2 [{shapeType=Capsule, length=0.505, radius=0.505, shapeName=pumpkin2_capsule, rotation=<rot 0 -90 0>}]
 pumpkin3 [{shapeType=Capsule, length=0.505, radius=0.505, shapeName=pumpkin3_capsule, rotation=<rot 0 -90 0>}]
 pumpkin_dyn1 [{shapeType=Capsule, length=1, radius=3, shapeName=pumpkin_dyn1_capsule, center=<0 0 -5>}]
 pumpkin_dyn3 [{shapeType=Capsule, length=1, radius=3, shapeName=pumpkin_dyn3_capsule, center=<0 0 -5>}]
 pumpkin_dyn_2 [{shapeType=Capsule, length=1, radius=3, shapeName=pumpkin_dyn_2_capsule, center=<0 0 -5>}]
}
constraints [
 {
  bone1Name=pumpkin1
  bone2Name=pelvis
  swing1LimitDegrees=45
  swing1Motion=Free
  swing2LimitDegrees=45
  swing2Motion=Free
  twistLimitDegrees=45
  twistMotion=Free
 }
 {
  bone1Name=pumpkin2
  bone2Name=pelvis
  swing1LimitDegrees=45
  swing1Motion=Free
  swing2LimitDegrees=45
  swing2Motion=Free
  twistLimitDegrees=45
  twistMotion=Free
 }
 {
  bone1Name=pumpkin3
  bone2Name=pelvis
  swing1LimitDegrees=45
  swing1Motion=Free
  swing2LimitDegrees=45
  swing2Motion=Free
  twistLimitDegrees=45
  twistMotion=Free
 }
 {bone1Name=pumpkin_dyn1, bone2Name=pumpkin1, swing1LimitDegrees=0, swing2LimitDegrees=20, twistLimitDegrees=0}
 {bone1Name=pumpkin_dyn_2, bone2Name=pumpkin2, swing1LimitDegrees=0, swing2LimitDegrees=20, twistLimitDegrees=0}
 {bone1Name=pumpkin_dyn3, bone2Name=pumpkin3, swing1LimitDegrees=0, swing2LimitDegrees=20, twistLimitDegrees=0}
]
