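# Post-process style Anim Blueprint for the M_MED_Foe player body, copying the parent mesh pose and adding secondary motion and corrections.
# Twist and sculpter Control Rigs feed pelvis-space RigidBody physics, a crouch pouch1 offset, then chest constraints, both targeting clavicle_l and spine_04.
# Despite its name, isSkyDiving enables physics only when neither diving nor floating.
# The pouch2, crouch-sprint and world-space RigidBody nodes are unconnected to output.
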AnimBlueprint /BRCosmetics/Characters/Player/Male/Medium/Bodies/M_MED_Foe/Meshes/M_MED_Foe_AnimBP
target_skeleton=&/M_MED_Foe_Skeleton
graphs [
 AnimGraph {nodes [
   n1 = Get Is Crouching() -> (Is Crouching:double)
   n2 = "Transform (Modify) Bone - Bone: pouch2"(Alpha=@n1, Translation="-1.75,-2.5,1", Rotation="0,-15,0", Scale="1,1,1") {
    rotationMode=BMM_Additive
    scaleMode=BMM_Additive
    translationMode=BMM_Additive
    alphaBoolBlend {blendOption=Linear}
    alphaScaleBiasClamp {bInterpResult=true, interpSpeedDecreasing=30, interpSpeedIncreasing=30}
    boneToModify {boneName=pouch2}
    rotation {yaw=-15}
    translation=<-1.75 -2.5 1>
   }
   "Transform (Modify) Bone - Bone: pouch1"(ComponentPose=@n2, Alpha=Get Is Crouch Sprinting() -> (Is Crouch Sprinting:double), Translation="0,0,1", Rotation="0,0,0", Scale="1,1,1") {
    rotationMode=BMM_Additive
    scaleMode=BMM_Additive
    translationMode=BMM_Additive
    alphaBoolBlend {blendOption=Linear}
    alphaScaleBiasClamp {bInterpResult=true, interpSpeedDecreasing=30, interpSpeedIncreasing=30}
    boneToModify {boneName=pouch1}
    translation {z=1}
   }
   RigidBody() {
    bClampLinearTranslationLimitToRefPose=true
    bOverrideWorldGravity=true
    overridePhysicsAsset=&/M_MED_Foe_Physics
    simulationSpace=WorldSpace
    alphaBoolBlend {blendOption=Linear}
    baseBoneRef {boneName=pelvis}
    componentAppliedLinearAccClamp=<3000 3000 3000>
    componentLinearAccScale=<0.7 0.6 0.6>
    componentLinearVelScale=<0.7 0.6 0.6>
    overrideWorldGravity {z=-980}
   }
   n3 = Control Rig(Source=Copy Pose From Mesh() {bUseAttachedParent=true}) {
    alphaBoolBlend {blendOption=Linear}
    controlRigAssetReference {blueprintRigClass=&/M_MED_Foe_CtrlRig_Twist.M_MED_Foe_CtrlRig_Twist_C}
   }
   n4 = Control Rig!(
    Source=@n3
    Alpha
    KneeRot_fr_l_114deg:Vector
    KneeTrans_fr_l_114deg:Vector="0, 0, 0"
    KneeScale_fr_l_114deg:Vector="0, 0, 0"
    KneeScale_fr_l_130deg:Vector="0, 0, 0"
    KneeRot_fr_l_130deg:Vector="0, 0, 0"
    KneeTrans_fr_l_130deg:Vector="0, 0, 0"
   ) -> (Pose) {
    alphaBoolBlend {blendOption=Linear}
    controlRigAssetReference {blueprintRigClass=&/M_MED_Foe_CtrlRig_sculpters.M_MED_Foe_CtrlRig_sculpters_C}
   }
   n5 = RigidBody!(
    ComponentPose=Local To Component(LocalPose=@n4)
    Alpha=Get isSkyDiving() -> (isSkyDiving:double)
    OverrideWorldGravity:Vector="0,0,-980"
    ExternalForce
   ) -> (Pose) {
    bClampLinearTranslationLimitToRefPose=true
    bOverrideWorldGravity=true
    overridePhysicsAsset=&M_MED_Foe_Physics
    simulationSpace=BaseBoneSpace
    alphaBoolBlend {blendOption=Linear}
    baseBoneRef {boneName=pelvis}
    componentAppliedLinearAccClamp=<3000 3000 3000>
    overrideWorldGravity {z=-980}
    simSpaceSettings {
     maxAngularAcceleration=1000
     maxAngularVelocity=3
     maxLinearAcceleration=250
     maxLinearVelocity=500
     velocityScaleZ=0
     worldAlpha=1
     externalLinearDragV {y=0.5}
    }
   }
   n6 = "Transform (Modify) Bone - Bone: pouch1"(ComponentPose=@n5, Alpha=@n1, Translation="-8,2,0", Rotation="0,35,0", Scale="1,1,1") {
    rotationMode=BMM_Additive
    scaleMode=BMM_Additive
    translationMode=BMM_Additive
    alphaBoolBlend {blendOption=Linear}
    alphaScaleBiasClamp {bInterpResult=true, interpSpeedDecreasing=30, interpSpeedIncreasing=30}
    boneToModify {boneName=pouch1}
    rotation {yaw=35}
    translation=<-8 2>
   }
   n7 = "Constraint - chest_l (TR)"!(ComponentPose=@n6, Alpha=1, ConstraintWeights_0=0, ConstraintWeights_1:float=1) -> (Pose) {
    alphaBoolBlend {blendOption=Linear}
    boneToModify {boneName=chest_l}
    constraintSetup [{transformType=Translation, targetBone {boneName=clavicle_l}}, {transformType=Rotation, targetBone {boneName=spine_04}}]
    constraintWeights [1,1]
   }
   n8 = "Constraint - chest_r (TR)"!(ComponentPose=@n7, Alpha=1, ConstraintWeights_0=0, ConstraintWeights_1:float=1) -> (Pose) {
    alphaBoolBlend {blendOption=Linear}
    boneToModify {boneName=chest_r}
    constraintSetup [{transformType=Translation, targetBone {boneName=clavicle_l}}, {transformType=Rotation, targetBone {boneName=spine_04}}]
    constraintWeights [1,1]
   }
   Output Pose(Result=Component To Local(ComponentPose=@n8)) {layerGroup=None}
  ]}
 EventGraph {nodes [
   Event BlueprintUpdateAnimation()
   n1 = Get bIsCrouching()
   > Set Is Crouching(Is Crouching:double="To Float (Boolean)"(InBool=@n1)) -> (Output_Get:double)
   n2 = Get bIsMoving2D()
   > Set Is Crouch Sprinting(Is Crouch Sprinting:double="To Float (Boolean)"(InBool=AND Boolean(A=@n1, B=@n2))) -> (Output_Get:double)
   n3 = NOT Boolean(A=Get bIsSkydiveDiveMode())
   n4 = NOT Boolean(A=Get bIsSkydiveFloating())
   > Set isSkyDiving(isSkyDiving:double="To Float (Boolean)"(InBool=AND Boolean(A=@n4, B=@n3))) -> (Output_Get:double)
   Get bIsSkyDiving()
  ]}
]
variables [
 Is Crouch Sprinting {type=number}
 Is Crouching {type=number}
 isSkyDiving {type=number}
]
